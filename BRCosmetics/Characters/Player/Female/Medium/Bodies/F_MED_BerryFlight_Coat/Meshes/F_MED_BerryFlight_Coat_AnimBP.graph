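# Animation Blueprint for the female BerryFlight coat cosmetic, adding secondary motion to the pose copied from the parent mesh.
# A Control Rig feeds RigidBody simulation, then weighted constraints place bag, ring, zipper and flap base bones.
# bPbBlendEnabled toggles layered glute dynamics before AnimDynamics swings foot rings, neck flap and zippers.
# The event graph computes a gravity override and enables the rig's jacket scaling at LOD 2 and above.
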AnimBlueprint /BRCosmetics/Characters/Player/Female/Medium/Bodies/F_MED_BerryFlight_Coat/Meshes/F_MED_BerryFlight_Coat_AnimBP
target_skeleton=&/F_MED_BerryFlight_Coat_Skeleton
graphs [
 AnimGraph {nodes [
   n1 = Use cached pose 'PreDyn'()
   n2 = Use cached pose 'GluteDynamics_1_1'()
   n3 = Get bPbBlendEnabled() -> (bPbBlendEnabled:bool)
   n4 = Anim Dynamics - Bone: foot_ring_bk_l(
    ComponentPose=Local To Component(LocalPose=Blend Poses by bool(BlendPose_0=@n2, BlendPose_1=@n1, bActiveValue=@n3))
    Alpha=1
    ExternalForce="0,0,0"
   ) {
    angularDampingOverride=1
    angularSpringConstant=500
    bAngularSpring=true
    bOverrideAngularDamping=true
    bUseGravityOverride=true
    linearDampingOverride=1
    linearSpringConstant=50
    simulationSpace=RootRelative
    alphaBoolBlend {blendOption=Linear}
    boundBone {boneName=foot_ring_bk_l}
    componentAppliedLinearAccClamp=<3000 3000 3000>
    componentLinearAccScale=<0.4 0.4 0.4>
    componentLinearVelScale=<0.4 0.4 0.4>
    gravityOverride {z=-980}
    physicsBodyDefinitions [{
      boundBone {boneName=foot_ring_bk_l}
      boxExtents=<10 10 10>
      constraintSetup {angularLimitsMax=<0 0 0>, angularLimitsMin=<0 0 -70>, angularTarget=<1 0 0>}
      localJointOffset=<0 3 0>
     }]
   }
   n5 = Anim Dynamics - Bone: foot_ring_bk_r(ComponentPose=@n4, Alpha=1, ExternalForce="0,0,0") {
    angularDampingOverride=1
    angularSpringConstant=500
    bAngularSpring=true
    bOverrideAngularDamping=true
    bUseGravityOverride=true
    linearDampingOverride=1
    linearSpringConstant=50
    simulationSpace=RootRelative
    alphaBoolBlend {blendOption=Linear}
    boundBone {boneName=foot_ring_bk_r}
    componentAppliedLinearAccClamp=<3000 3000 3000>
    componentLinearAccScale=<0.4 0.4 0.4>
    componentLinearVelScale=<0.4 0.4 0.4>
    gravityOverride {z=-980}
    physicsBodyDefinitions [{
      boundBone {boneName=foot_ring_bk_r}
      boxExtents=<10 10 10>
      constraintSetup {angularLimitsMax=<0 0 0>, angularLimitsMin=<0 0 -70>, angularTarget=<1 0 0>}
      localJointOffset=<0 3 0>
     }]
   }
   n6 = Anim Dynamics - Bone: dyn_flap_neck_r!(ComponentPose=@n5, Alpha=1, GravityOverride:Vector="0,0,-980") -> (Pose) {
    angularDampingOverride=1
    angularSpringConstant=500
    bAngularSpring=true
    bOverrideAngularDamping=true
    bUseGravityOverride=true
    gravityScale=0.35
    simulationSpace=RootRelative
    alphaBoolBlend {blendOption=Linear}
    alphaScaleBiasClamp {interpSpeedDecreasing=2.5, interpSpeedIncreasing=8}
    boundBone {boneName=dyn_flap_neck_r}
    componentAppliedLinearAccClamp=<3000 3000 3000>
    componentLinearAccScale=<0.4 0.4 0.4>
    componentLinearVelScale=<0.3 0.3 0.3>
    gravityOverride {z=-980}
    physicsBodyDefinitions [{
      boundBone {boneName=dyn_flap_neck_r}
      boxExtents=<15 15 15>
      constraintSetup {angularLimitsMax=<0 8 0>, angularLimitsMin=<0 -8 0>, angularTarget=<1 0 0>}
      localJointOffset=<3 0 0>
     }]
    relativeSpaceBone {boneName=root}
    simSpaceSettings {maxAngularAcceleration=35, maxAngularVelocity=7, simSpaceAngularAlpha=1}
   }
   n7 = Anim Dynamics - Bone: clavicle_zipper_l(ComponentPose=@n6, Alpha=1, ExternalForce="0,0,0") {
    angularDampingOverride=1
    angularSpringConstant=500
    bAngularSpring=true
    bOverrideAngularDamping=true
    bUseGravityOverride=true
    linearDampingOverride=1
    linearSpringConstant=50
    simulationSpace=RootRelative
    alphaBoolBlend {blendOption=Linear}
    boundBone {boneName=clavicle_zipper_l}
    componentAppliedLinearAccClamp=<3000 3000 3000>
    componentLinearAccScale=<0.4 0.4 0.4>
    componentLinearVelScale=<0.4 0.4 0.4>
    gravityOverride {z=-980}
    physicsBodyDefinitions [{
      boundBone {boneName=clavicle_zipper_l}
      boxExtents=<10 10 10>
      constraintSetup {angularLimitsMax=<0 5 0>, angularLimitsMin=<0 -5 -70>, angularTarget=<1 0 0>}
      localJointOffset=<3 0 0>
     }]
   }
   n8 = Anim Dynamics - Bone: clavicle_zipper_r(ComponentPose=@n7, Alpha=1, ExternalForce="0,0,0") {
    angularDampingOverride=1
    angularSpringConstant=500
    bAngularSpring=true
    bOverrideAngularDamping=true
    bUseGravityOverride=true
    linearDampingOverride=1
    linearSpringConstant=50
    simulationSpace=RootRelative
    alphaBoolBlend {blendOption=Linear}
    boundBone {boneName=clavicle_zipper_r}
    componentAppliedLinearAccClamp=<3000 3000 3000>
    componentLinearAccScale=<0.4 0.4 0.4>
    componentLinearVelScale=<0.4 0.4 0.4>
    gravityOverride {z=-980}
    physicsBodyDefinitions [{
      boundBone {boneName=clavicle_zipper_r}
      boxExtents=<10 10 10>
      constraintSetup {angularLimitsMax=<0 5 0>, angularLimitsMin=<0 -5 -70>, angularTarget=<1 0 0>}
      localJointOffset=<3 0 0>
     }]
   }
   n9 = Anim Dynamics - Bone: spine_zipper_l(ComponentPose=@n8, Alpha=1, ExternalForce="0,0,0") {
    angularDampingOverride=1
    angularSpringConstant=500
    bAngularSpring=true
    bOverrideAngularDamping=true
    bUseGravityOverride=true
    linearDampingOverride=1
    linearSpringConstant=50
    simulationSpace=RootRelative
    alphaBoolBlend {blendOption=Linear}
    boundBone {boneName=spine_zipper_l}
    componentAppliedLinearAccClamp=<3000 3000 3000>
    componentLinearAccScale=<0.4 0.4 0.4>
    componentLinearVelScale=<0.4 0.4 0.4>
    gravityOverride {z=-980}
    physicsBodyDefinitions [{
      boundBone {boneName=spine_zipper_l}
      boxExtents=<10 10 10>
      constraintSetup {angularLimitsMax=<0 5 0>, angularLimitsMin=<0 -5 -70>, angularTarget=<1 0 0>}
      localJointOffset=<3 0 0>
     }]
   }
   Output Pose(Result=Component To Local(ComponentPose=@n9)) {layerGroup=None}
   n10 = "Transform (Modify) Bone - Bone: deform_glute_l"(ComponentPose=Local To Component(LocalPose=Use cached pose 'PreDyn'()), Alpha=1, Translation="0,-1,0", Rotation="0,0,0", Scale="1,1,1") {translationMode=BMM_Additive, translationSpace=BCS_BoneSpace, boneToModify {boneName=deform_glute_l}, translation {y=-1}}
   n11 = "Transform (Modify) Bone - Bone: deform_glute_r"(ComponentPose=@n10, Alpha=1, Translation="0,1,0", Rotation="0,0,0", Scale="1,1,1") {translationMode=BMM_Additive, translationSpace=BCS_BoneSpace, boneToModify {boneName=deform_glute_r}, translation {y=1}}
   n12 = Get Sim Space Settings() -> (Sim Space Settings:Anim Phys Sim Space Settings Structure)
   n13 = Anim Dynamics - Bone: deform_glute_r!(ComponentPose=@n11, Alpha=1, AngularSpringConstant:float=5000, SimSpaceSettings:Anim Phys Sim Space Settings Structure=@n12) -> (Pose) {
    angularBiasOverride=1
    angularDampingOverride=0.9999
    angularSpringConstant=5000
    bAngularSpring=true
    bGravityOverrideInSimSpace=true
    bOverrideAngularBias=true
    bOverrideAngularDamping=true
    bUseGravityOverride=true
    linearDampingOverride=0.9999
    linearSpringConstant=100
    simulationSpace=RootRelative
    boundBone {boneName=deform_glute_r}
    componentAppliedLinearAccClamp=<2 2 2>
    gravityOverride {z=-980}
    physicsBodyDefinitions [{
      boundBone {boneName=deform_glute_r}
      boxExtents=<21 21 21>
      constraintSetup {angularLimitsMax=<0 5 5>, angularLimitsMin=<0 -5 -5>, angularTarget=<1 0 0>}
      localJointOffset=<2 0 0>
     }]
    relativeSpaceBone {boneName=spine_04}
    simSpaceSettings {maxAngularAcceleration=45, maxAngularVelocity=8, simSpaceAngularAlpha=0.25}
   }
   n14 = Anim Dynamics - Bone: deform_glute_l!(ComponentPose=@n13, Alpha=1, AngularSpringConstant:float=5000, SimSpaceSettings:Anim Phys Sim Space Settings Structure=@n12) -> (Pose) {
    angularBiasOverride=1
    angularDampingOverride=0.9999
    angularSpringConstant=5000
    bAngularSpring=true
    bGravityOverrideInSimSpace=true
    bOverrideAngularBias=true
    bOverrideAngularDamping=true
    bUseGravityOverride=true
    linearDampingOverride=0.9999
    linearSpringConstant=100
    simulationSpace=RootRelative
    boundBone {boneName=deform_glute_l}
    componentAppliedLinearAccClamp=<2 2 2>
    gravityOverride {z=-980}
    physicsBodyDefinitions [{
      boundBone {boneName=deform_glute_l}
      boxExtents=<20 20 20>
      constraintSetup {angularLimitsMax=<0 5 5>, angularLimitsMin=<0 -5 -5>, angularTarget=<1 0 0>}
      localJointOffset=<-2 0 0>
     }]
    relativeSpaceBone {boneName=spine_04}
    simSpaceSettings {maxAngularAcceleration=45, maxAngularVelocity=8, simSpaceAngularAlpha=0.25}
   }
   n15 = Anim Dynamics - Bone: deform_glute_r!(ComponentPose=@n14, Alpha=1, GravityOverride:Vector="0,0,-200", LinearSpringConstant:float=180) -> (Pose) {
    angularBiasOverride=1
    angularDampingOverride=0.99
    angularSpringConstant=10000
    bGravityOverrideInSimSpace=true
    bLinearSpring=true
    bOverrideAngularBias=true
    bOverrideLinearDamping=true
    bUseGravityOverride=true
    linearDampingOverride=1
    linearSpringConstant=180
    simulationSpace=RootRelative
    boundBone {boneName=deform_glute_r}
    componentAppliedLinearAccClamp=<5 5 5>
    gravityOverride {z=-200}
    physicsBodyDefinitions [{
      boundBone {boneName=deform_glute_r}
      boxExtents=<80 80 80>
      constraintSetup {angularLimitsMax=<0 0 0>, angularLimitsMin=<0 0 0>, angularTarget=<1 0 0>, linearAxesMax=<2 2 2>, linearAxesMin=<-2 -2 -2>}
      localJointOffset=<-0.1 0.1 0>
     }]
    relativeSpaceBone {boneName=spine_04}
    simSpaceSettings {maxAngularAcceleration=50, maxAngularVelocity=10, simSpaceAngularAlpha=0.1}
   }
   n16 = Anim Dynamics - Bone: deform_glute_l!(ComponentPose=@n15, Alpha=1, GravityOverride:Vector="0,0,-200", LinearSpringConstant:float=180) -> (Pose) {
    angularBiasOverride=1
    angularDampingOverride=0.99
    angularSpringConstant=10000
    bGravityOverrideInSimSpace=true
    bLinearSpring=true
    bOverrideAngularBias=true
    bOverrideLinearDamping=true
    bUseGravityOverride=true
    linearDampingOverride=1
    linearSpringConstant=180
    simulationSpace=RootRelative
    boundBone {boneName=deform_glute_l}
    componentAppliedLinearAccClamp=<5 5 5>
    gravityOverride {z=-200}
    physicsBodyDefinitions [{
      boundBone {boneName=deform_glute_l}
      boxExtents=<90 90 90>
      constraintSetup {angularLimitsMax=<0 0 0>, angularLimitsMin=<0 0 0>, angularTarget=<1 0 0>, linearAxesMax=<2 2 2>, linearAxesMin=<-2 -2 -2>}
      localJointOffset=<0.1 -0.1 0>
     }]
    relativeSpaceBone {boneName=spine_04}
    simSpaceSettings {maxAngularAcceleration=50, maxAngularVelocity=10, simSpaceAngularAlpha=0.1}
   }
   Save cached pose 'GluteDynamics_1_1'(Pose=Component To Local(ComponentPose=@n16)) {cacheName=GluteDynamics_1_1}
   n17 = RigidBody!(ComponentPose=Local To Component(LocalPose=Use cached pose 'CtrlRig'()), Alpha, OverrideWorldGravity:Vector="0,0,0") -> (Pose) {
    bClampLinearTranslationLimitToRefPose=true
    bOverrideWorldGravity=true
    overridePhysicsAsset=&/F_MED_BerryFlight_Coat_Physics
    simulationSpace=BaseBoneSpace
    alphaBoolBlend {blendOption=Linear}
    baseBoneRef {boneName=spine_04}
    componentAppliedLinearAccClamp=<0 0 0>
    simSpaceSettings {
     maxAngularAcceleration=40
     maxAngularVelocity=10
     maxLinearAcceleration=1000
     maxLinearVelocity=300
     velocityScaleZ=0.4
     worldAlpha=1
     externalLinearDragV=<0.3 0.3 0.3>
    }
   }
   n18 = "Constraint - bag_l_base (PPPP)"!(
    ComponentPose=@n17
    Alpha=1
    ConstraintWeights_0=0.0009
    ConstraintWeights_1:float=0.4387
    ConstraintWeights_2:float=0.5599
    ConstraintWeights_3:float=0.0005
   ) -> (Pose) {
    boneToModify {boneName=bag_l_base}
    constraintSetup [
     {targetBone {boneName=pelvis}}
     {targetBone {boneName=thigh_twist_01_l}}
     {targetBone {boneName=thigh_l}}
     {targetBone {boneName=deform_glute_l}}
    ]
    constraintWeights [1,1,1,1]
   }
   n19 = "Constraint - foot_ring_bk_base_l (PPPPPPPP)"!(
    ComponentPose=@n18
    Alpha=1
    ConstraintWeights_0=0
    ConstraintWeights_1:float=0
    ConstraintWeights_2:float=0.5374
    ConstraintWeights_3:float=0.3086
    ConstraintWeights_4:float=0.0002
    ConstraintWeights_5:float=0.0017
    ConstraintWeights_6:float=0
    ConstraintWeights_7:float=0.152
   ) -> (Pose) {
    boneToModify {boneName=foot_ring_bk_base_l}
    constraintSetup [
     {targetBone {boneName=thigh_twist_01_l}}
     {targetBone {boneName=thigh_l}}
     {targetBone {boneName=calf_twist_01_l}}
     {targetBone {boneName=foot_l}}
     {targetBone {boneName=calf_l}}
     {targetBone {boneName=calf_twist_02_l}}
     {targetBone {boneName=def_foot_corr_bk_r}}
     {targetBone {boneName=def_foot_corr_bk_l}}
    ]
    constraintWeights [1,1,1,1,1,1,1,1]
   }
   n20 = "Constraint - foot_ring_bk_base_r (PPPPPPPP)"!(
    ComponentPose=@n19
    Alpha=1
    ConstraintWeights_0=0
    ConstraintWeights_1:float=0
    ConstraintWeights_2:float=0.3086
    ConstraintWeights_3:float=0.5374
    ConstraintWeights_4:float=0.0002
    ConstraintWeights_5:float=0.0017
    ConstraintWeights_6:float=0.152
    ConstraintWeights_7:float=0
   ) -> (Pose) {
    boneToModify {boneName=foot_ring_bk_base_r}
    constraintSetup [
     {targetBone {boneName=thigh_twist_01_r}}
     {targetBone {boneName=thigh_r}}
     {targetBone {boneName=foot_r}}
     {targetBone {boneName=calf_twist_01_r}}
     {targetBone {boneName=calf_r}}
     {targetBone {boneName=calf_twist_02_r}}
     {targetBone {boneName=def_foot_corr_bk_r}}
     {targetBone {boneName=def_foot_corr_bk_l}}
    ]
    constraintWeights [1,1,1,1,1,1,1,1]
   }
   n21 = "Constraint - clavicle_zipper_base_l (PPPPPPPP)"!(
    ComponentPose=@n20
    Alpha=1
    ConstraintWeights_0=0.2803
    ConstraintWeights_1:float=0.1204
    ConstraintWeights_2:float=0.2491
    ConstraintWeights_3:float=0.0295
    ConstraintWeights_4:float=0
    ConstraintWeights_5:float=0.0248
    ConstraintWeights_6:float=0
    ConstraintWeights_7:float=0.2959
   ) -> (Pose) {
    boneToModify {boneName=clavicle_zipper_base_l}
    constraintSetup [
     {targetBone {boneName=spine_05}}
     {targetBone {boneName=clavicle_l}}
     {targetBone {boneName=spine_04}}
     {targetBone {boneName=upperarm_twist_01_l}}
     {targetBone {boneName=spine_02}}
     {targetBone {boneName=deform_upperarm_fr_l}}
     {targetBone {boneName=deform_upperarm_up_l}}
     {targetBone {boneName=deform_pec_l}}
    ]
    constraintWeights [1,1,1,1,1,1,1,1]
   }
   n22 = "Constraint - clavicle_zipper_base_r (PPPPPPPP)"!(
    ComponentPose=@n21
    Alpha=1
    ConstraintWeights_0=0.2817
    ConstraintWeights_1:float=0.117
    ConstraintWeights_2:float=0.2512
    ConstraintWeights_3:float=0.0282
    ConstraintWeights_4:float=0
    ConstraintWeights_5:float=0.0229
    ConstraintWeights_6:float=0
    ConstraintWeights_7:float=0.299
   ) -> (Pose) {
    boneToModify {boneName=clavicle_zipper_base_r}
    constraintSetup [
     {targetBone {boneName=spine_05}}
     {targetBone {boneName=clavicle_r}}
     {targetBone {boneName=spine_04}}
     {targetBone {boneName=upperarm_twist_01_r}}
     {targetBone {boneName=spine_02}}
     {targetBone {boneName=deform_upperarm_fr_r}}
     {targetBone {boneName=deform_upperarm_up_r}}
     {targetBone {boneName=deform_pec_r}}
    ]
    constraintWeights [1,1,1,1,1,1,1,1]
   }
   n23 = "Constraint - spine_zipper_base (PPPPPPPP)"!(
    ComponentPose=@n22
    Alpha=1
    ConstraintWeights_0=0.0006
    ConstraintWeights_1:float=0.239
    ConstraintWeights_2:float=0.0326
    ConstraintWeights_3:float=0.0537
    ConstraintWeights_4:float=0.0069
    ConstraintWeights_5:float=0.0061
    ConstraintWeights_6:float=0.002
    ConstraintWeights_7:float=0.659
   ) -> (Pose) {
    boneToModify {boneName=spine_zipper_base}
    constraintSetup [
     {targetBone {boneName=spine_05}}
     {targetBone {boneName=spine_04}}
     {targetBone {boneName=spine_03}}
     {targetBone {boneName=spine_02}}
     {targetBone {boneName=spine_01}}
     {targetBone {boneName=pelvis}}
     {targetBone {boneName=deform_pec_l}}
     {targetBone {boneName=dyn_jacket_l_0}}
    ]
    constraintWeights [1,1,1,1,1,1,1,1]
   }
   n24 = "Constraint - base_flap_neck_r (PPPPPPPPP)"!(
    ComponentPose=@n23
    Alpha=1
    ConstraintWeights_0=0.6185
    ConstraintWeights_1:float=0.0251
    ConstraintWeights_2:float=0.0506
    ConstraintWeights_3:float=0.0014
    ConstraintWeights_4:float=0.0026
    ConstraintWeights_5:float=0.0169
    ConstraintWeights_6:float=0.007
    ConstraintWeights_7:float=0.2707
    ConstraintWeights_8:float=0.0072
   ) -> (Pose) {
    boneToModify {boneName=base_flap_neck_r}
    constraintSetup [
     {targetBone {boneName=spine_05}}
     {targetBone {boneName=clavicle_r}}
     {targetBone {boneName=spine_04}}
     {targetBone {boneName=upperarm_twist_01_r}}
     {targetBone {boneName=deform_upperarm_fr_r}}
     {targetBone {boneName=deform_upperarm_up_r}}
     {targetBone {boneName=deform_pec_r}}
     {targetBone {boneName=dyn_neck_front_r}}
     {targetBone {boneName=dyn_neck_side_r}}
    ]
    constraintWeights [1,1,1,1,1,1,1,1,1]
   }
   Save cached pose 'PreDyn'(Pose=Component To Local(ComponentPose=@n24)) {cacheName=PreDyn}
   n25 = Control Rig!(
    Source=Copy Pose From Mesh() {bUseAttachedParent=true}
    Alpha
    IsSkydiving:bool=Get bIsSkyDiving()
    IsTargeting:bool=Get bIsTargeting()
    IsScaleJacket:bool=Get ScaleJacket() -> (ScaleJacket:bool)
   ) -> (Pose) {controlRigAssetReference {blueprintRigClass=&/F_MED_BerryFlight_Coat_CtrlRig.F_MED_BerryFlight_Coat_CtrlRig_C}}
   Save cached pose 'CtrlRig'(Pose=@n25) {cacheName=CtrlRig}
  ]}
 EventGraph {nodes [
   Event BlueprintUpdateAnimation()
   > n1 = Sequence()
   > n2 = "Gravity Override State Function (Parameter Asset) (Nativized)"(
    CharacterPartAnimInstance=Self-Reference()
    GameTimeInSeconds=GetGameTimeInSeconds()
    Parameters=&/F_MED_BerryFlight_Coat_Parameters
   )
   > Set GravityOverride(GravityOverride=@n2)
   Set ScaleJacket(execute=@n1.then_1, ScaleJacket:bool="integer >= integer"(A=Get CurrentLOD(), B=2)) -> (Output_Get:bool)
  ]}
]
variables [
 GravityOverride {type=Vector}
 ScaleJacket {type=boolean}
 Sim Space Settings {
  type=AnimPhysSimSpaceSettings
  default {maxAngularAcceleration=60, maxAngularVelocity=12, simSpaceAngularAlpha=0.7, externalAngularVelocity=<0 0 0>}
 }
]
